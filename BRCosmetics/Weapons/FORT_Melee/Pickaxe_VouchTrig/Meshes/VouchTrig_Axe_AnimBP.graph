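# Animation Blueprint for the VouchTrig axe pickaxe, posing its Melee_TwoHanded_Handle bone to suit the wielder's body type.
# A state machine leaves a borrowed DonkeyCribUniversal idle pose for a reference-pose Swing during melee animations, then feeds DefaultSlot.
# On customization load, a character part's body and gender set Large Character for large males.
# That flag chooses between two additive bone adjustments: move, rotate and scale, or rotation only.
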AnimBlueprint /BRCosmetics/Weapons/FORT_Melee/Pickaxe_VouchTrig/Meshes/VouchTrig_Axe_AnimBP
target_skeleton=&/VouchTrig_Axe_Skeleton
graphs [
 AnimGraph {nodes [
   AnimGraphNode_StateMachine_0 = Pickaxe() {is=title AnimGraphNode_StateMachine}
   n1 = Get Large Character() -> (Large Character:bool)
   n2 = "Transform (Modify) Bone - Bone: Melee_TwoHanded_Handle"!(
    ComponentPose=Local To Component(LocalPose=Slot 'DefaultSlot'(Source=@AnimGraphNode_StateMachine_0))
    bAlphaBoolEnabled:bool=@n1
    Translation="1,0,4"
    Rotation="0,0,-10.5"
    Scale="1.2,1.2,1"
   ) -> (Pose) {
    alphaInputType=Bool
    rotationMode=BMM_Additive
    rotationSpace=BCS_ParentBoneSpace
    scaleMode=BMM_Additive
    scaleSpace=BCS_BoneSpace
    translationMode=BMM_Additive
    translationSpace=BCS_BoneSpace
    alphaBoolBlend {blendOption=Linear}
    boneToModify {boneName=Melee_TwoHanded_Handle}
    rotation {roll=-10.5}
    scale=<1.2 1.2>
    translation {x=1, z=4}
   }
   n3 = "Transform (Modify) Bone - Bone: Melee_TwoHanded_Handle"(ComponentPose=@n2, bAlphaBoolEnabled=NOT Boolean(A=@n1), Translation="0,0,0", Rotation="6.565,-0.855,-7.390", Scale="1,1,1") {alphaInputType=Bool, rotationMode=BMM_Additive, boneToModify {boneName=Melee_TwoHanded_Handle}, rotation=<rot 6.565 -0.855 -7.39>}
   Output Pose(Result=Component To Local(ComponentPose=@n3)) {layerGroup=None}
  ]}
 Pickaxe {
  entry=Idle
  path=AnimGraph.AnimGraphNode_StateMachine_0.Pickaxe
  ids=id
  nodes [
   AnimStateNode_0 = Idle()
   AnimStateNode_1 = Swing() {is=title AnimStateNode}
   AnimStateTransitionNode_1 = AnimStateTransitionNode()
  ]
  links [{dst=Swing, rule=AnimGraph.AnimGraphNode_StateMachine_0.Pickaxe.AnimStateTransitionNode_1.Transition, src=Idle, via=AnimStateTransitionNode_1}]
 }
 Idle {
  path=AnimGraph.AnimGraphNode_StateMachine_0.Pickaxe.AnimStateNode_0.Idle
  nodes [
   n1 = Sequence Player 'DonkeyCribUniversal_Equip_IdlePose_Axe'() {bLoopAnimation=false, sequence=/BRCosmetics/Weapons/FORT_Melee/Pickaxe_DonkeyCribUniversal/Anim/DonkeyCribUniversal_Equip_IdlePose_Axe}
   Output Animation Pose(Result=@n1)
  ]
 }
 Swing {path=AnimGraph.AnimGraphNode_StateMachine_0.Pickaxe.AnimStateNode_1.Swing, nodes [Output Animation Pose(Result=Local Space Ref Pose())]}
 Transition {
  path=AnimGraph.AnimGraphNode_StateMachine_0.Pickaxe.AnimStateTransitionNode_1.Transition
  nodes [Result(bCanEnterTransition=Get bIsPlayingMeleeAnim())]
 }
 EventGraph {nodes [
   n1 = Event OnCharacterAndPickaxeCustomizationLoaded!() -> (
    OutputDelegate
    then
    CustomBaseMesh:Skeletal Mesh Component Object Reference
    CustomParts:Array of Custom Character Part Object References
    CustomMeshParts:Array of Skeletal Mesh Component Budgeted Object References
   ) {is=title K2Node_Event}
   n2 = Get Array Item with EFort Custom Part Type(Array:Array of Custom Character Part Object References=@n1.CustomParts) -> (Output:Custom Character Part Object Reference)
   n3 = Get BodyTypesPermitted!(self:Custom Character Part Object Reference=@n2) -> (BodyTypesPermitted:EFortCustomBodyType Enum=NONE)
   n4 = Get GenderPermitted!(self:Custom Character Part Object Reference=@n2) -> (GenderPermitted:EFortCustomGender Enum=Invalid)
   > Set Large Character(Large Character:bool=AND Boolean(
     A="Equal (Enum)"(A:EFortCustomBodyType Enum=@n3, B:EFortCustomBodyType Enum=Large)
     B="Equal (Enum)"(A:EFortCustomGender Enum=@n4, B:EFortCustomGender Enum=Male)
    )) -> (Output_Get:bool)
   TryGetPawnOwner()
   Event BlueprintUpdateAnimation()
  ]}
]
variables [Large Character {type=boolean}]
